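# Physics asset for the Donut medium male backpack cosmetic, defining simple physics for the pack and its two swords.
# Three box bodies cover the pack and the left and right dynamic sword bones.
# Each sword joins to the pack through its own constraint with matching angular limits, allowing the widest movement on one swing axis, less on the other and very little twist.
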PhysicsAsset /BRCosmetics/Accessories/FORT_Backpacks/Backpack_M_MED_Donut/Meshes/Donut_Physics
bodies {
 dyn_sword_l [{shapeType=Box, extentX=20, extentY=3.6018, extentZ=9.2502, center=<-1 0.3156 0>}]
 dyn_sword_r [{shapeType=Box, extentX=20, extentY=3.6018, extentZ=6.7372, center=<-1 -0.3156 0>}]
 pack [{shapeType=Box, extentX=15.5958, extentY=3.8105, extentZ=14.9352, center=<0 -5.1311 -1.8494>}]
}
constraints [
 {bone1Name=dyn_sword_l, bone2Name=pack, swing1LimitDegrees=15, swing2LimitDegrees=45, twistLimitDegrees=5}
 {bone1Name=dyn_sword_r, bone2Name=pack, swing1LimitDegrees=15, swing2LimitDegrees=45, twistLimitDegrees=5}
]
