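# Animation Blueprint for the Donut backpack cosmetic, adding secondary physics motion on its skeleton.
# The input pose gets an additive roll on the root bone, with blend alpha driven by the yaw of the backpack rotation.
# A RigidBody node then simulates the Donut physics asset relative to the root, with overridden gravity and partial component linear acceleration.
# The event graph is empty, so all logic lives in the AnimGraph.
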AnimBlueprint /BRCosmetics/Accessories/FORT_Backpacks/Backpack_M_MED_Donut/Meshes/Donut_Skeleton_AnimBlueprint
target_skeleton=&/Donut_Skeleton
graphs [
 AnimGraph {nodes [
   n1 = "Transform (Modify) Bone - Bone: root"(
    ComponentPose=Local To Component(LocalPose=Input Pose())
    Alpha=BreakRotator(InRot=Get BackpackRotation() -> (BackpackRotation:Rotator)).Yaw
    Translation="0,0,0"
    Rotation="0,0,10"
   ) {
    rotationMode=BMM_Additive
    rotationSpace=BCS_BoneSpace
    alphaBoolBlend {blendOption=Linear}
    boneToModify {boneName=root}
    rotation {roll=10}
   }
   n2 = RigidBody(ComponentPose=@n1) {
    bClampLinearTranslationLimitToRefPose=true
    bOverrideWorldGravity=true
    overridePhysicsAsset=&/Donut_Physics
    simulationSpace=BaseBoneSpace
    alphaBoolBlend {blendOption=Linear}
    baseBoneRef {boneName=root}
    componentLinearAccScale=<0.5 0.5 0.5>
    overrideWorldGravity {z=-980}
   }
   Output Pose(Result=Component To Local(ComponentPose=@n2)) {layerGroup=None}
  ]}
 EventGraph {links [], nodes []}
]
variables []
